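# Physics asset for the AirKringle female medium head cosmetic, setting up physical motion for its carrot piece, likely a nose.
# It defines capsule bodies for the head and two dyn_carrot bones.
# Constraints chain the second carrot bone to the first and the first to the head, with tight angular limits that allow a small wobble, slightly looser on one swing axis at the outer joint.
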PhysicsAsset /BRCosmetics/Characters/Player/Female/Medium/Heads/F_MED_AirKringle_Head/Meshes/F_MED_AirKringle_Head_Physics
bodies {
 dyn_carrot_01 [{
   shapeType=Capsule
   length=0.762
   radius=0.7137
   shapeName=dyn_carrot_01_capsule
   center=<1.0123 -0.0086 -0.015>
   rotation=<rot 0 89.8426 -84.1561>
  }]
 dyn_carrot_02 [{
   shapeType=Capsule
   length=3.6329
   radius=3.1349
   shapeName=dyn_carrot_02_capsule
   center=<3.4393 0.0229 -0.3761>
   rotation=<rot 0 -91.3659 -80.455>
  }]
 head [{shapeType=Capsule, length=17.6181, radius=13.3502, shapeName=head_capsule, center=<11.0594 7.404 -0.0069>, rotation=<rot 0 4.9521 -89.9451>}]
}
constraints [
 {bone1Name=dyn_carrot_01, bone2Name=head, swing1LimitDegrees=15, swing2LimitDegrees=15, twistLimitDegrees=15}
 {bone1Name=dyn_carrot_02, bone2Name=dyn_carrot_01, swing1LimitDegrees=15, swing2LimitDegrees=35, twistLimitDegrees=15}
]
