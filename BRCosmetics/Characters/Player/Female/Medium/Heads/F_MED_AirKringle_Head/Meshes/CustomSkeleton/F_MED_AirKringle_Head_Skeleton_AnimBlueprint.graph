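# Animation blueprint for the AirKringle female medium player head cosmetic, adding physics-driven secondary motion to its custom skeleton.
# The input pose runs through a RigidBody node using the head's physics asset at partial alpha, with overridden gravity and damped component motion, then goes to output.
# Two chained spring-based Anim Dynamics nodes for the carrot bones are left disconnected from the output.
# The event graph holds only unconnected stub nodes.
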AnimBlueprint /BRCosmetics/Characters/Player/Female/Medium/Heads/F_MED_AirKringle_Head/Meshes/CustomSkeleton/F_MED_AirKringle_Head_Skeleton_AnimBlueprint
target_skeleton=&/F_MED_AirKringle_Head_Skeleton
graphs [
 AnimGraph {nodes [
   n1 = Anim Dynamics - Bone: dyn_carrot_01(Alpha=1, ExternalForce="0,0,0") {
    angularDampingOverride=1
    angularSpringConstant=20000
    bAngularSpring=true
    bOverrideAngularDamping=true
    bUseGravityOverride=true
    bUsePlanarLimit=false
    gravityScale=0.35
    simulationSpace=RootRelative
    alphaBoolBlend {blendOption=Linear}
    boundBone {boneName=dyn_carrot_01}
    componentLinearAccScale=<1 1 1>
    componentLinearVelScale=<1 1 1>
    gravityOverride {z=-980}
    physicsBodyDefinitions [{
      boundBone {boneName=dyn_carrot_01}
      boxExtents=<20 30 30>
      constraintSetup {angularLimitsMax=<15 15 15>, angularLimitsMin=<-15 -15 -15>, angularTarget=<1 0 0>}
      localJointOffset=<4 0 0>
     }]
    relativeSpaceBone {boneName=head}
   }
   Anim Dynamics - Bone: dyn_carrot_02(ComponentPose=@n1, Alpha=1, ExternalForce="0,0,0") {
    angularDampingOverride=1
    angularSpringConstant=20000
    bAngularSpring=true
    bOverrideAngularDamping=true
    bUseGravityOverride=true
    bUsePlanarLimit=false
    gravityScale=0.35
    simulationSpace=RootRelative
    alphaBoolBlend {blendOption=Linear}
    boundBone {boneName=dyn_carrot_02}
    componentLinearAccScale=<1 1 1>
    componentLinearVelScale=<1 1 1>
    gravityOverride {z=-980}
    physicsBodyDefinitions [{
      boundBone {boneName=dyn_carrot_02}
      boxExtents=<20 30 30>
      constraintSetup {angularLimitsMax=<15 15 15>, angularLimitsMin=<-15 -20 -15>, angularTarget=<1 0 0>}
      localJointOffset=<4 0 0>
     }]
    relativeSpaceBone {boneName=head}
   }
   n2 = RigidBody(ComponentPose=Local To Component(LocalPose=Input Pose()), Alpha=0.6) {
    bOverrideWorldGravity=true
    overridePhysicsAsset=/BRCosmetics/Characters/Player/Female/Medium/Heads/F_MED_AirKringle_Head/Meshes/F_MED_AirKringle_Head_Physics
    alphaBoolBlend {blendOption=Linear}
    componentLinearAccScale=<0.6 0.6 0.6>
    componentLinearVelScale=<0.6 0.6 0.6>
    overrideWorldGravity {z=-980}
    simSpaceSettings {maxAngularAcceleration=250, maxAngularVelocity=250, maxLinearAcceleration=250, maxLinearVelocity=250, velocityScaleZ=0.2}
   }
   Output Pose(Result=Component To Local(ComponentPose=@n2))
  ]}
 EventGraph {nodes [
   TryGetPawnOwner()
   Event BlueprintUpdateAnimation()
  ]}
]
variables []
